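# Physics asset for the Shark Attack backpack cosmetic, giving its pack, knife and jet bones simple collision bodies for physics motion.
# Boxes cover the pack, knife and jet master, while small spheres cover the left and right jets.
# Angular-limited constraints link the knife to the pack and each jet to the jet master, with tighter twist than swing.
# No constraint links the jet master to the pack.
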PhysicsAsset /BRCosmetics/Accessories/FORT_Backpacks/Backpack_M_MED_Shark_Attack/Meshes/M_MED_Shark_Attack_Physics
bodies {
 dyn_jet_l [{shapeType=Sphere, radius=3.1127, center=<1 0 0>}]
 dyn_jet_mstr [{shapeType=Box, extentX=4.304, extentY=4.304, extentZ=4.304}]
 dyn_jet_r [{shapeType=Sphere, radius=3.1356, center=<1 0 0>}]
 dyn_knife [{shapeType=Box, extentX=2.269, extentY=5.2253, extentZ=25.0503, center=<1.1574 -0.3711 -0.5938>, rotation=<rot 90 0 0>}]
 pack [{shapeType=Box, extentX=15.8404, extentY=15.1549, extentZ=37.142, center=<-0.3135 -5.8733 -9.2663>, rotation=<rot 0 -89.9996 -10.0474>}]
}
constraints [
 {bone1Name=dyn_knife, bone2Name=pack, swing1LimitDegrees=15, swing2LimitDegrees=15, twistLimitDegrees=5}
 {bone1Name=dyn_jet_r, bone2Name=dyn_jet_mstr, swing1LimitDegrees=15, swing2LimitDegrees=15, twistLimitDegrees=5}
 {bone1Name=dyn_jet_l, bone2Name=dyn_jet_mstr, swing1LimitDegrees=15, swing2LimitDegrees=15, twistLimitDegrees=5}
]
